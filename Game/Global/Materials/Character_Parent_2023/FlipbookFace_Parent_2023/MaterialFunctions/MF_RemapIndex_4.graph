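# Material function in the flipbook face parent material set that remaps a scalar index through a four-entry lookup table.
# The Index is wrapped by fmod 4 and drives a Switch that selects one component of the Input_0-3 vector, split via BreakOutFloat4Components.
# The selected component is output as Remapped Index, with the vector defaulting to an identity order.
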MaterialFunction /Game/Global/Materials/Character_Parent_2023/FlipbookFace_Parent_2023/MaterialFunctions/MF_RemapIndex_4
graph {expressions {
  MaterialExpressionConstant_0 {class=MaterialExpressionConstant, properties {r=4}}
  MaterialExpressionFmod_0 {
   class=MaterialExpressionFmod
   inputs [
    A {from=MaterialExpressionFunctionInput_0, output=""}
    B {from=MaterialExpressionConstant_0, output=""}
   ]
   properties {}
  }
  MaterialExpressionFunctionInput_0 {
   class=MaterialExpressionFunctionInput
   inputs [Preview {from="", output=""}]
   properties {bUsePreviewValueAsDefault=true, inputName=Index, inputType=FunctionInput_Scalar}
  }
  MaterialExpressionFunctionInput_1 {
   class=MaterialExpressionFunctionInput
   inputs [Preview {from="", output=""}]
   properties {bUsePreviewValueAsDefault=true, inputName=Input_0-3, inputType=FunctionInput_Vector4, sortPriority=1, previewValue=<0 1 2 3>}
  }
  MaterialExpressionFunctionOutput_0 {class=MaterialExpressionFunctionOutput, inputs [None {from=MaterialExpressionSwitch_3, output=""}], properties {outputName=Remapped Index}}
  MaterialExpressionMaterialFunctionCall_0 {
   class=MaterialExpressionMaterialFunctionCall
   inputs [Float4 {from=MaterialExpressionFunctionInput_1}]
   properties {materialFunction=&BreakOutFloat4Components}
  }
  MaterialExpressionSwitch_3 {
   class=MaterialExpressionSwitch
   inputs [
    SwitchValue {from=MaterialExpressionFmod_0, output=""}
    Default {from="", output=""}
    None {from=MaterialExpressionMaterialFunctionCall_0, output=R}
    None {from=MaterialExpressionMaterialFunctionCall_0, output=R}
    None {from=MaterialExpressionMaterialFunctionCall_0, output=R}
    None {from=MaterialExpressionMaterialFunctionCall_0, output=R}
   ]
   properties {inputs [
     {
      inputName=None
      input {
       expression=&self:MaterialExpressionMaterialFunctionCall_0
       inputName=None
       mask=0
       maskA=0
       maskB=0
       maskG=0
       maskR=0
       outputIndex=0
      }
     }
     {
      inputName=None
      input {
       expression=&self:MaterialExpressionMaterialFunctionCall_0
       inputName=None
       mask=0
       maskA=0
       maskB=0
       maskG=0
       maskR=0
       outputIndex=1
      }
     }
     {
      inputName=None
      input {
       expression=&self:MaterialExpressionMaterialFunctionCall_0
       inputName=None
       mask=0
       maskA=0
       maskB=0
       maskG=0
       maskR=0
       outputIndex=2
      }
     }
     {
      inputName=None
      input {
       expression=&self:MaterialExpressionMaterialFunctionCall_0
       inputName=None
       mask=0
       maskA=0
       maskB=0
       maskG=0
       maskR=0
       outputIndex=3
      }
     }
    ]}
  }
 }}
