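# Parent anime-shaded material for the ElixirFret cosmetic companion, using the Cloth shading model.
# Shared Pre, Calculations and Post functions build the toon look, with normal-sphere and eyeball line-masking functions feeding in.
# Masked eye animation and a flipbook mouth, indexed by thresholded animation curves, supply face detail with optional color, emissive and opacity application.
# Pose and look-direction change flags, plus eye-mask terms, drive temporal responsiveness.
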Material /CosmeticCompanions/Assets/Biped/ElixirFret/Materials/M_AnimeShader_ElixirFret_Parent
graph {
 properties {
  bAutomaticallySetUsageInEditor=false
  bUseMaterialAttributes=true
  bUsedWithClothing=true
  bUsedWithMeshDeformer=true
  bUsedWithMorphTargets=true
  bUsedWithNiagaraMeshParticles=true
  bUsedWithSkeletalMesh=true
  bUsedWithStaticLighting=true
  pixelDepthOffsetMode=PDOM_Legacy
  shadingModel=MSM_Cloth
  usedShadingModels=Cloth
 }
 nodes [
  n1 = ScalarParameter() {
   controlType=Enumeration
   enumeration=/Game/Global/Materials/Enums/E_Mat_OnOff
   group=Flipbook 01 Global
   parameterName=Flipbook.ApplyOpacityMask
   sortPriority=4
  }
  NamedRerouteDeclaration(None=LinearInterpolate(A=Constant(), B=NamedRerouteUsage(), Alpha=@n1)) {name=Flipbook_Opacity, nodeColor=<rgba 0 1 0.6016 1>}
  n2 = NamedRerouteUsage()
  n3 = Multiply(A=NamedRerouteUsage(), B=ScalarParameter() {defaultValue=1, group=Flipbook 01 Global, parameterName=Flipbook.EmissiveStrength})
  n4 = NamedRerouteUsage()
  n5 = ScalarParameter() {
   controlType=Enumeration
   enumeration=&E_Mat_OnOff
   group=Flipbook 01 Global
   parameterName=Flipbook.ApplyEmissive
   sortPriority=3
  }
  NamedRerouteDeclaration(None=LinearInterpolate(A=NamedRerouteUsage(), B=LinearInterpolate(A=@n2, B=@n3, Alpha=@n4), Alpha=@n5)) {name=Flipbook_Emissive, nodeColor=<rgba 0.108 0.2905 1 1>}
  n6 = TextureSampleParameter2D() {
   group=01_EyeAnimation_Parameters
   parameterName=EyeMask
   samplerSource=SSM_Wrap_WorldGroupSettings
   samplerType=SAMPLERTYPE_Masks
   sortPriority=-32
   texture=&T_WhiteMask
  }
  n7 = VertexColor()
  n8 = AppendVector(A=@n7, B=@n7)
  n9 = ScalarParameter() {
   controlType=Enumeration
   enumeration=&E_Mat_OnOff
   group=01_EyeAnimation_Parameters
   parameterName=UseVertexColorForEyeMask
   sortPriority=-33
  }
  n10 = ComponentMask(None=TextureCoordinate() {coordinateIndex=1}) {r=true}
  n11 = Saturate(None=Sign(None=Subtract(A=@n10) {constB=0.0001}))
  n12 = StaticSwitchParameter(
   True=ChannelMaskParameter(None=LinearInterpolate(A=@n6.RGBA, B=@n8, Alpha=@n9)) {group=01_EyeAnimation_Parameters, parameterName=EyeMaskChannel, sortPriority=-32}.R
   False=@n11
  ) {group=01_EyeAnimation_Parameters, parameterName=UseEyeMask, sortPriority=-33}
  n13 = MaterialFunctionCall!(
   Color Lit=NamedRerouteUsage()
   Color Shaded=NamedRerouteUsage()
   Use UV2=StaticBoolParameter() {defaultValue=true, group=01_EyeAnimation_Parameters, parameterName=Use UV2 for Eye Animation, sortPriority=-30}
   Mask=@n12
  ) {materialFunction=/CosmeticCompanions/Materials/MaterialFunctions/MF_CompanionAnimeEyeAnimation}
  NamedRerouteDeclaration(None=@n13.Shaded) {name=EeyAnimColor_Shaded, nodeColor=<rgba 0.125 1 0 1>}
  n14 = NamedRerouteUsage()
  n15 = Multiply(A=NamedRerouteUsage(), B=ScalarParameter() {defaultValue=1, parameterName=FlipbookFaceAnimeShadedValue})
  n16 = NamedRerouteUsage()
  n17 = ScalarParameter() {
   controlType=Enumeration
   defaultValue=1
   enumeration=&E_Mat_OnOff
   group=Flipbook 01 Global
   parameterName=Flipbook.ApplyBaseColor
   sortPriority=2
  }
  NamedRerouteDeclaration(None=LinearInterpolate(A=NamedRerouteUsage(), B=LinearInterpolate(A=@n14, B=@n15, Alpha=@n16), Alpha=@n17)) {name=Flipbook_ColorShaded, nodeColor=<rgba 1 0 0.0937 1>}
  n18 = NamedRerouteUsage()
  n19 = NamedRerouteUsage()
  n20 = NamedRerouteUsage()
  n21 = ScalarParameter() {
   controlType=Enumeration
   defaultValue=1
   enumeration=&E_Mat_OnOff
   group=Flipbook 01 Global
   parameterName=Flipbook.ApplyBaseColor
   sortPriority=2
  }
  NamedRerouteDeclaration(None=LinearInterpolate(A=NamedRerouteUsage(), B=LinearInterpolate(A=@n18, B=@n19, Alpha=@n20), Alpha=@n21)) {name=Flipbook_ColorLit, nodeColor=<rgba 0 0.4141 1 1>}
  n22 = MaterialFunctionCall!(CustomUV) {materialFunction=/Game/Global/Materials/Character_Parent_2023/Anime_Parent_2023/MaterialFunctions/MFC_AnimeShader_2023_Pre}
  NamedRerouteDeclaration(None=@n22.Color_Lit) {name=Color_Lit, nodeColor=<rgba 0 1 0.8359 1>}
  NamedRerouteDeclaration(None=StaticBoolParameter() {defaultValue=true, group=Flipbook 01 Global, parameterName=Flipbook.UseMouth, sortPriority=-8}) {name=UseMouth, nodeColor=<rgba 0.1953 1 0 1>}
  NamedRerouteDeclaration(None=@n22.Color_Shaded) {name=Color_Shaded, nodeColor=<rgba 1 0.5156 0 1>}
  n23 = MaterialFunctionCall!(
   Curve Index=NamedRerouteUsage()
   Use Texture=NamedRerouteUsage()
   ----------------------
   Vertex Color=NamedRerouteUsage()
   ---------------------
   Flipbook Texture=NamedRerouteUsage()
   Column Count=NamedRerouteUsage()
   Row Count=NamedRerouteUsage()
   UVs=NamedRerouteUsage()
   UV Offset=NamedRerouteUsage()
   UV Scale=NamedRerouteUsage()
  ) {materialFunction=/Game/Global/Materials/Character_Parent_2023/FlipbookFace_Parent_2023/MaterialFunctions/MF_Sample_Flipbook}
  NamedRerouteDeclaration(None=@n23.Flipbook Vertex Mask) {name=MouthVertexMask, nodeColor=<rgba 0.2092 0.8266 1 1>}
  NamedRerouteDeclaration(None=@n23.Flipbook Color) {name=MouthColor, nodeColor=<rgba 0.2092 0.8266 1 1>}
  NamedRerouteDeclaration(None=@n13.Lit) {name=EyeAnimColor_Lit, nodeColor=<rgba 0.1953 0 1 1>}
  n24 = VectorParameter() {
   group=Flipbook 04 Mouth
   parameterName=Flipbook.MouthRemap_0-3
   sortPriority=201
   channelNames {a="3 - Jaw Open ", b=2 - Frown, g=1 - Smile, r=0 - Neutral}
   defaultValue=<rgba 0 1 2 3>
  }
  NamedRerouteDeclaration(None=@n24.RGBA) {name=Mouth_Remap_0-3, nodeColor=<rgba 1 0.6652 0.0014 1>}
  n25 = ScalarParameter() {group="Flipbook 04.01 Mouth | Curves (Driven By Animation)", parameterName=mouth_eh_pose_extra, sortPriority=6}
  n26 = ScalarParameter() {defaultValue=0.1, group=Flipbook 04.02 Mouth | Thresholds, parameterName=Threshold_Eh, sortPriority=6}
  n27 = Constant() {r=6}
  n28 = MaterialFunctionCall() {materialFunction=&/MF_SK_Flipbook_Mouth_Curves_ElixirFret}
  n29 = NamedRerouteUsage()
  n30 = Min(
   A=If(
    A=ScalarParameter() {group="Flipbook 04.01 Mouth | Curves (Driven By Animation)", parameterName=R_smile_pose, sortPriority=0}
    B=ScalarParameter() {defaultValue=0.1, group=Flipbook 04.02 Mouth | Thresholds, parameterName=Threshold_Smile, sortPriority=0}
    "A > B"=Constant() {r=1}
    "A < B"=Constant()
   )
   B=If(
    A=ScalarParameter() {group="Flipbook 04.01 Mouth | Curves (Driven By Animation)", parameterName=jaw_open_pose, sortPriority=2}
    B=ScalarParameter() {defaultValue=0.1, group=Flipbook 04.02 Mouth | Thresholds, parameterName=Threshold_Jaw_Open, sortPriority=2}
    "A > B"=Constant() {r=1}
    "A < B"=Constant()
   )
  )
  n31 = If(
   A=ScalarParameter() {group="Flipbook 04.01 Mouth | Curves (Driven By Animation)", parameterName=mouth_uh_pose_extra, sortPriority=5}
   B=ScalarParameter() {defaultValue=0.1, group=Flipbook 04.02 Mouth | Thresholds, parameterName=Threshold_Uh, sortPriority=5}
   "A > B"=Constant() {r=5}
   "A < B"=LinearInterpolate(
    A=MaterialFunctionCall!(Index=@n28.Index, Input_0-3=@n29) {materialFunction=/Game/Global/Materials/Character_Parent_2023/FlipbookFace_Parent_2023/MaterialFunctions/MF_RemapIndex_4}.Remapped Index
    Alpha=NamedRerouteDeclaration(None=@n30) {name=OpenSmile_Threshold, nodeColor=<rgba 0 1 0.9766 1>}
   ) {constB=4}
  )
  NamedRerouteDeclaration(None=If(
    A=ScalarParameter() {group="Flipbook 04.01 Mouth | Curves (Driven By Animation)", parameterName=mouth_ah_pose_extra, sortPriority=7}
    B=ScalarParameter() {defaultValue=0.1, group=Flipbook 04.02 Mouth | Thresholds, parameterName=Threshold_Ah, sortPriority=7}
    "A > B"=Constant() {r=7}
    "A < B"=If(A=@n25, B=@n26, "A > B"=@n27, "A < B"=@n31)
   )) {name=Mouth_CurveIndex, nodeColor=<rgba 1 0.4453 0 1>}
  NamedRerouteDeclaration(None=ChannelMaskParameter(None=VertexColor()) {group=Flipbook 04 Mouth, parameterName=Flipbook.MouthVertexColor}.R) {name=Mouth_VertexColor, nodeColor=<rgba 1 0.9844 0 1>}
  NamedRerouteDeclaration(None=StaticBoolParameter() {defaultValue=true, group=Flipbook 04 Mouth, parameterName=Flipbook.UseTextureForMouth, sortPriority=-20}) {name=UseTextureForMouth, nodeColor=<rgba 0 1 0.3438 1>}
  n32 = TextureObjectParameter() {
   group=Flipbook 04 Mouth
   parameterName=Flipbook.MouthTexture
   sortPriority=2
   texture=/Game/Global/Materials/Character_Parent_2023/FlipbookFace_Parent_2023/Textures/T_black_emptyalpha
  }
  NamedRerouteDeclaration(None=@n32) {name=Mouth_FlipbookTexture, nodeColor=<rgba 1 0.1875 0 1>}
  NamedRerouteDeclaration(None=ScalarParameter() {defaultValue=3, group=Flipbook 04 Mouth, parameterName=Flipbook.MouthRowCount, sortPriority=3}) {name=Mouth_RowCount, nodeColor=<rgba 1 0 0.7734 1>}
  NamedRerouteDeclaration(None=ScalarParameter() {defaultValue=3, group=Flipbook 04 Mouth, parameterName=Flipbook.MouthColumnCount, sortPriority=4}) {name=Mouth_ColumnCount, nodeColor=<rgba 1 0 0.7734 1>}
  NamedRerouteDeclaration(None=StaticSwitchParameter(True=TextureCoordinate() {coordinateIndex=1}, False=TextureCoordinate()) {group=Flipbook 04 Mouth, parameterName=Flipbook.MouthUseUV2, sortPriority=10}) {name=Mouth_UVs, nodeColor=<rgba 1 0.5859 0 1>}
  n33 = ScalarParameter() {defaultValue=1, group=Flipbook 04 Mouth, parameterName=Flipbook.MouthUVScaleX, sortPriority=12}
  n34 = ScalarParameter() {defaultValue=1, group=Flipbook 04 Mouth, parameterName=Flipbook.MouthUVScaleY, sortPriority=13}
  NamedRerouteDeclaration(None=Multiply(
    A=ScalarParameter() {defaultValue=1, group=Flipbook 04 Mouth, parameterName=Flipbook.MouthGlobalScale, sortPriority=12}
    B=AppendVector(A=@n33, B=@n34)
   )) {name=Mouth_UVScale, nodeColor=<rgba 0.2656 0 1 1>}
  NamedRerouteDeclaration(None=AppendVector(
    A=ScalarParameter() {group=Flipbook 04 Mouth, parameterName=Flipbook.MouthUVOffsetX, sortPriority=20}
    B=ScalarParameter() {group=Flipbook 04 Mouth, parameterName=Flipbook.MouthUVOffsetY, sortPriority=21}
   )) {name=Mouth_UVOffset, nodeColor=<rgba 0.0547 1 0 1>}
  NamedRerouteDeclaration(None=StaticSwitch(
    True=NamedRerouteDeclaration(None=@n23.Flipbook Alpha) {name=MouthAlpha, nodeColor=<rgba 0.2092 0.8266 1 1>}
    False=Constant()
    Value=NamedRerouteUsage()
   )) {name=Flipbook_FaceMask, nodeColor=<rgba 1 0.7969 0 1>}
  NamedRerouteDeclaration(None=StaticSwitch(True=NamedRerouteUsage(), False=Constant(), Value=NamedRerouteUsage())) {name=Flipbook_FaceColor, nodeColor=<rgba 1 0.849 0.0986 1>}
  NamedRerouteDeclaration(None=@n22.Emissive) {name=Emissive, nodeColor=<rgba 1 0.1875 0 1>}
  n35 = MaterialFunctionCall!(
   CustomLightVector
   OverrideNormal_WS=MaterialFunctionCall!(Normal=@n22.Normal) {materialFunction=/CosmeticCompanions/Materials/MaterialFunctions/MF_CompanionAnime_NormalSphere}.Normal
   STT=@n22.STT
   Normal_Map=@n22.STT
   DFL=@n22.STT
   FX=@n22.STT
  ) {materialFunction=/Game/Global/Materials/Character_Parent_2023/Anime_Parent_2023/MaterialFunctions/MFC_AnimeShader_2023_Calculations}
  n36 = GetMaterialAttributes(None=MaterialFunctionCall!(
    Shading_Gradient=@n35.Shading_Gradient
    Threshold_Gradient=@n35.Shading_Gradient
    Specular_Gradient=@n35.Shading_Gradient
    Sheen_Mask=@n35.Shading_Gradient
    Thickness=@n35.Shading_Gradient
    DFL_Mask=MaterialFunctionCall!(DFLines=@n35.DFL_Mask) {materialFunction=&/MF_EyeballsDFLineMasking_ElixirFret}.Result
    FX=@n35.Shading_Gradient
    Metallic_Gradient=@n35.Shading_Gradient
    RimLight_Gradient=@n35.Shading_Gradient
    NormalBuffer=@n35.Shading_Gradient
    --------
    CINE_LightSources=@n35.CINE_LightSources
    CINE_LightColor=@n35.CINE_LightColor
    ---------
    Color_Lit=NamedRerouteUsage()
    Color_Shaded=NamedRerouteUsage()
    SRM=@n22.SRM
    Emissive=NamedRerouteUsage()
    DFL_Color=@n22.DFL_Color
    Specular_Color=@n22.Specular_Color
    RimLight_Color=@n22.StyleRimLight_Color
   ) {materialFunction=/Game/Global/Materials/Character_Parent_2023/Anime_Parent_2023/MaterialFunctions/MFC_AnimeShader_2023_Post}.Result) {attributeGetTypes [#1, #2]}
  SetMaterialAttributes!(
   MaterialAttributes=NamedRerouteDeclaration(None=@n36.MaterialAttributes) {name=Material, nodeColor=<rgba 0.6406 0 1 1>}
   Base Color=@n36.Base Color
   Opacity Mask=NamedRerouteUsage()
   World Position Offset=@n36.Base Color
  ) {attributeSetTypes [#1, #3, #2]}
  n37 = ScalarParameter() {parameterName=HasPoseChanged}
  n38 = ScalarParameter() {parameterName=HasLookDirectionChanged}
  n39 = NamedRerouteUsage()
  n40 = NamedRerouteUsage()
  n41 = LinearInterpolate(Alpha=Multiply(A=Max(A=@n37, B=@n38), B=Max(A=@n39, B=@n40))) {constA=0.4}
  n42 = Subtract(A=@n11, B=NamedRerouteDeclaration(None=@n12) {name=EyeMask, nodeColor=<rgba 1 0 0.4219 1>})
  TemporalResponsivenessOutput!(None=Saturate(None=Add(A=@n41, B=@n42)))
 ]
}
